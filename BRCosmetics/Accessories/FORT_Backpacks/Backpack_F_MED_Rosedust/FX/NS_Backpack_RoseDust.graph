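# Niagara effect for the Rosedust backpack cosmetic, showing a glitching graphic sprite with bursts of orbs, glitch shapes, rings and digital debris.
# A system Timeline value drives the graphic's glitch parameters and gates the GPU Digital_Bits and Noise spawns.
# User.StateColorInt selects palettes and ring count and scale, while AdditionalElementInt switches on weakpoint splatter or damage streaks.
# User offset, size-switch and pivot inputs fit placement and shape.
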NiagaraSystem /BRCosmetics/Accessories/FORT_Backpacks/Backpack_F_MED_Rosedust/FX/NS_Backpack_RoseDust
dependencies {
 usedDataInterfaces [
  &NiagaraDataInterfaceCurve
  &NiagaraDataInterfaceArrayFloat2
  &NiagaraDataInterfaceArrayColor
  &NiagaraDataInterfaceCamera
  &NiagaraDataInterfaceArrayInt32
  &NiagaraDataInterfaceArrayFloat
  &NiagaraDataInterfaceVector2DCurve
 ]
 usedDynamicInputs []
 usedModules [&SystemState, &Timeline]
}
emitters {
 Digital_Bits {
  simTarget=GPUComputeSim
  emitterUpdateScript [
   &EmitterState()
   &SpawnRate(Spawn Probability=$System.Timeline.Float, SpawnRate=45, Use Spawn Probability=true)
  ]
  particleSpawnScript [
   /SetVariables(Particles.ChromaOffset=1) {isSetParameters=true}
   /Niagara/Modules/Spawn/Initialization/InitializeParticle(
    Color=<rgba 1 1 5 1>
    Lifetime=&UniformRangedFloat(Minimum=0.05, Maximum=0.1)
    Module.Write Color=true
    Module.Write Lifetime=true
    Module.Write Mass=true
    Module.Write Material Random=false
    Module.Write Position=false
    Module.Write Scale=false
    Module.Write SpriteRotation=false
    Module.Write SpriteSize=true
    Position=&SimulationPosition()
    Sprite Size=&RandomRangeVector2D(Minimum=<1 0.25>, Maximum=<2 0.5>)
   )
   /SetVariables(
    Particles.TextureSampleA:LinearColor=$Output.SampleTexture.SampledColor
    Particles.TextureSampleB:LinearColor=$Output.SampleTexture001.SampledColor
   ) {isSetParameters=true}
   &ShapeLocation(
    Box Size=&MakeCustomVectorFromBool(Bool=$User.NonUniformSizeSwitch, True Vector=<30 30 30>, False Vector=<25 25 15>)
    MaskSpawnGroup=false
    Offset=$User.Offset
    Offset Mode=::NewEnumerator0
    Override Seed=false
    Shape Primitive=::NewEnumerator2
    Surface Only=false
    UseOwnerScale=true
   ) {module=ShapeLocation001}
   &CameraOffset(Camera Offset Amount=&Lerp_Float(
     A=20
     B=4
     Alpha=&FloatFromCurve(FloatCurve=DataInterface() {propertyValues {Curve {keys [<at 0 1>,<at 1 0>]}}}, CurveIndex=$Emitter.Scalability.DistanceFraction)
    ))
   &AddVelocity(Rotation Coordinate Space=::World, Velocity=<0 0 -20>) {module=AddVelocity001}
  ]
  particleUpdateScript [
   &ParticleState()
   &AddVelocity(Velocity=&Multiply_Vector(A=$Particles.NormalVelocity, B=&VectorFromFloat(Value=-1)))
   &SolveForcesAndVelocity(Clamp Velocity=false, Limit Acceleration=false)
   /SetVariables() {enabled=false, isSetParameters=true}
   /SetVariables(Particles.ChromaOffset:NiagaraFloat=&FloatFromCurve(
     FloatCurve=DataInterface() {propertyValues {Curve {keys [{interpMode=RCIM_Cubic, tangentMode=RCTM_User, time=0, value=1}, {interpMode=RCIM_Cubic, tangentMode=RCTM_User, time=1, value=0}]}}}
     Scale Curve=0.5
    )) {isSetParameters=true}
   /SetVariables(
    Particles.Color001:LinearColor=<rgba 0.1 0 0 0.5>
    Particles.Position001:NiagaraPosition=&ConvertVectorToPosition(
     Input Position=&Add_Vector(
      A=&ConvertPositionToVector(Input Position=$Particles.Position, Convert To Absolute Large World Vector=false)
      B=&Multiply_VectorByFloat(Vector=<0 0 0>, Float=$Particles.ChromaOffset)
     )
     Vector Is In Absolute Large World Space:NiagaraBool=false
    )
   ) {isSetParameters=true}
   /SetVariables(
    Particles.Color002=<rgba 0 0.0077 0.1 0.5>
    Particles.Position002:NiagaraPosition=&ConvertVectorToPosition(
     Input Position=&Add_Vector(
      A=&ConvertPositionToVector(Input Position=$Particles.Position, Convert To Absolute Large World Vector=false)
      B=&Multiply_VectorByFloat(Vector=<0 0 0>, Float=$Particles.ChromaOffset)
     )
     Vector Is In Absolute Large World Space:NiagaraBool=false
    )
   ) {isSetParameters=true}
   /SetVariables(
    Particles.Color003:LinearColor=<rgba 0.0029 0.1 0 0.5>
    Particles.Position003:NiagaraPosition=&ConvertVectorToPosition(
     Input Position=&Add_Vector(
      A=&ConvertPositionToVector(Input Position=$Particles.Position, Convert To Absolute Large World Vector=false)
      B=&Multiply_VectorByFloat(Vector=<0 0 0>, Float=$Particles.ChromaOffset)
     )
     Vector Is In Absolute Large World Space:NiagaraBool=false
    )
   ) {isSetParameters=true}
   &ScaleSpriteSize(Scale Factor=&Vector2DFromCurve(
     Vector2Curve=DataInterface() {propertyValues {
       XCurve {keys [{tangentMode=RCTM_User, time=0, value=0}, {tangentMode=RCTM_User, time=0.5, value=10}, <at 1 0>]}
       YCurve {keys [{tangentMode=RCTM_User, time=0, value=0}, {tangentMode=RCTM_User, time=0.5, value=0.35}, <at 1 0.025>]}
      }}
     ScaleCurve=<2 8>
    ))
   /SetVariables(Particles.SpriteSize001:Vector2f=&Multiply_Vector2D_ByFloat(Float=1, Vector2D=$Particles.SpriteSize)) {isSetParameters=true}
  ]
  rendererClasses [&NiagaraSpriteRendererProperties]
  renderers [
   {rendererClass=&NiagaraSpriteRendererProperties}
   {rendererClass=&NiagaraSpriteRendererProperties, rendererIndex=1}
   {rendererClass=&NiagaraSpriteRendererProperties, rendererIndex=2}
   {rendererClass=&NiagaraSpriteRendererProperties, rendererIndex=3}
  ]
 }
 GlitchShapes {
  emitterUpdateScript [
   &EmitterState(Life Cycle Mode=::NewEnumerator1, Loop Behavior=::NewEnumerator1, Loop Duration=0.4, UseLoopDelay=false)
   &SpawnBurst_Instantaneous(Spawn Count=12, Use Spawn Probability=false, UseLoopCountLimit=true)
  ]
  particleSpawnScript [
   /SetVariables(Particles.RandomBone:NiagaraInt32=&RandomRangeInt(Minimum=0, Maximum=5)) {isSetParameters=true}
   /SetVariables(Particles.SizeLerp:NiagaraFloat=&FloatFromCurve(
     FloatCurve=DataInterface() {propertyValues {Curve {keys [
         {interpMode=RCIM_Cubic, tangentMode=RCTM_User, time=0, value=0}
         {arriveTangent=1.4324, interpMode=RCIM_Cubic, leaveTangent=1.4324, tangentMode=RCTM_User, time=0.87, value=0.5596}
         {interpMode=RCIM_Cubic, tangentMode=RCTM_User, time=1, value=1}
        ]}}}
     CurveIndex=&RandomRangeFloat(Minimum=0, Maximum=1)
    )) {isSetParameters=true}
   /SetVariables(Particles.BoneIndex:NiagaraInt32=&MaskIntBySpawnGroup(
     Masked Int=&RandomRangeInt(Minimum=0, Maximum=1)
     Passthrough Int=&SelectIntFromArray(Direct Array Index=$Particles.RandomBone, Int Selection Array=DataInterface() {propertyValues {IntData [0,1,5,6,9,10]}})
     Spawn Group Mask=1
    )) {isSetParameters=true}
   /SetVariables(Particles.StateColorInt:NiagaraInt32=$User.StateColorInt) {isSetParameters=true, module=SetVariables_#1}
   /Niagara/Modules/Spawn/Initialization/V2/InitializeParticle(
    Color=&MakeCustomLinearColorFromBool(
     Bool=&SetBoolByFloatComparison(A=$Particles.SizeLerp, B=0.1)
     True Color=&SelectLinearColorFromArray(
      Array Sampling Mode=::NewEnumerator1
      Direct Array Index=$Particles.StateColorInt
      Color Selection Array=DataInterface() {propertyValues {ColorData [<rgba 2 0 1.3623 1>, <rgba 0.1564 2 1.104 1>, <rgba 0.4768 1.0914 2 1>, <rgba 0.2928 1.8468 2 1>]}}
      Random Seed=?
     )
     False Color=&SelectLinearColorFromArray(
      Array Sampling Mode=::NewEnumerator1
      Direct Array Index=$Particles.StateColorInt
      Color Selection Array=DataInterface() {propertyValues {ColorData [<rgba 0 1.565 2 1>, <rgba 2 0.8358 0.0289 1>, <rgba 1.4198 2 0.0638 1>, <rgba 2 0.0603 0.4464 1>]}}
      Random Seed=?
     )
    )
    Lifetime Max=0.5
    Lifetime Min=0.3
    Lifetime Mode=::NewEnumerator1
    Mesh Scale=&Multiply_VectorByFloat(
     Vector=&RandomRangeVector(Minimum=<0.1 0.1 0.1>, Maximum=<0.2 0.2 0.2>)
     Float=&SelectFloatFromArray(Direct Array Index=$Particles.RandomBone, Float Selection Array=DataInterface() {propertyValues {FloatData [1.2,1.2,0.8,0.8,0.5,0.5]}})
    )
    Mesh Scale Mode=::NewEnumerator4
    Sprite Size Mode=::NewEnumerator3
    UsePositionOffset=false
    Write Material Random=true
   )
   &InitialMeshOrientation(Use Rotation Vector=false)
   &ShapeLocation(
    Box Size=&MakeCustomVectorFromBool(Bool=$User.NonUniformSizeSwitch, True Vector=<30 30 30>, False Vector=<40 40 25>)
    MaskSpawnGroup=false
    Offset=$User.Offset
    Offset Mode=::NewEnumerator0
    Override Seed=false
    Shape Primitive=::NewEnumerator2
    Surface Only=false
    UseOwnerScale=true
   )
   &AddVelocity(Module.Use Falloff Distance=false, Override Seed=false, Velocity Mode=::NewEnumerator1, Velocity Speed=150) {enabled=false}
   &CameraOffset(Camera Offset Amount=8)
  ]
  particleUpdateScript [
   &ParticleState()
   /SetVariables(Particles.Jitter:NiagaraFloat=&ClampFloat(
     Float=&Waveform(
      Global Clamp Min=-1
      "Waveform [1]"=::NewEnumerator12
      "[1] Frequency"=&RandomRangeFloat(Minimum=15, Maximum=25)
      "[1] Amplitude Min/Max"=<-150 150>
     )
     Min=-0.25
     Max=0.25
    )) {isSetParameters=true}
   &Drag(Drag=4, Use Linear Drag=true, Use Rotational Drag=false)
   &DynamicMaterialParameters(
    Index 0 Param 1=&Add_Float(A=$Particles.Jitter, B=&RandomRangeFloat(Minimum=0, Maximum=1))
    Index 0 Param 2=&Subtract_Float(A=&RandomRangeFloat(Minimum=0, Maximum=1), B=$Particles.Jitter)
    Index 0 Param 3=$Particles.Jitter
    Module.Param0WriteEnabled=true
    Module.Param1WriteEnabled=true
    Module.Param2WriteEnabled=true
    Module.Param3WriteEnabled=false
   )
   &ScaleColor(
    Scale Alpha=&Add_Float(A=1, B=&ClampFloat(Float=$Particles.Jitter, Min=0.8))
    ScaleA=true
    ScaleRGB=true
   )
   &UpdateMeshOrientation(Rotation Rate=0.2, Rotation Vector=&RandomRangeVector(Minimum=<0 0 0>, Maximum=<1 1 1>))
   &SolveForcesAndVelocity(Clamp Velocity=false, Limit Acceleration=false)
  ]
  rendererClasses [&NiagaraMeshRendererProperties]
  renderers [{rendererClass=&NiagaraMeshRendererProperties, rendererIndex=0}]
 }
 Graphic {
  emitterUpdateScript [
   &EmitterState(Life Cycle Mode=::NewEnumerator1, Loop Behavior=::NewEnumerator1, Loop Duration Mode=::NewEnumerator1, UseLoopDelay=false)
   &SpawnBurst_Instantaneous(Use Spawn Probability=false, UseLoopCountLimit=false)
   /SetVariables(
    Emitter.ChromaAmount:NiagaraFloat=&MakeCustomFloatFromBool(Bool=&SetBoolByFloatComparison(A=1, B=$System.Timeline.Float), True Float=0, False Float=1)
    Emitter.EffectChannelPicker:Vector3f=$User.EffectChannelPicker
    Emitter.FlipbookDistStrenght:NiagaraFloat=&MakeCustomFloatFromBool(Bool=&SetBoolByFloatComparison(A=1, B=$System.Timeline.Float), True Float=0, False Float=1)
    Emitter.FrontEndColorCorrect:NiagaraFloat=&FN_FrontEndFloatModifier(Gameplay=0.5, Front End=1)
    Emitter.GlitchStrenght:NiagaraFloat=&Multiply_Float(A=$System.Timeline.Float, B=1)
    Emitter.HDR Contrast:NiagaraFloat=&FN_FrontEndFloatModifier(Gameplay=0.5, Front End=1)
    Emitter.OffsetStrenght:NiagaraFloat=&MakeCustomFloatFromBool(Bool=$User.NonUniformSizeSwitch, True Float=0.2, False Float=0.1)
   ) {isSetParameters=true}
  ]
  particleSpawnScript [
   &InitializeParticle(
    Lifetime=0
    Position Offset=$User.Offset
    Sprite Size=<75 75>
    Sprite Size Mode=::NewEnumerator4
    UsePositionOffset=true
    Write Material Random=true
   )
   &CameraOffset(Camera Offset Amount=4)
  ]
  particleUpdateScript [
   &ParticleState(Kill Particles When Lifetime Has Elapsed=false, Let Infinitely Lived Particles Die When Emitter Deactivates=true)
   &ScaleSpriteSize(
    Scale Factor=&MakeCustomVector2DFromBool(Bool=$User.NonUniformSizeSwitch, True Vector2D=<0.6 0.55>, False Vector2D=<0.95 0.47>)
    Scale Sprite Size Mode=::NewEnumerator2
   ) {module=ScaleSpriteSize001}
   &ScaleSpriteSize(
    Scale Sprite Size Mode=::NewEnumerator0
    Uniform Scale Factor=&FloatFromCurve(
     FloatCurve=DataInterface() {propertyValues {Curve {keys [<at 0 0 RCIM_Constant>,<at 0.25 0.25 RCIM_Constant>,<at 0.5 0.5 RCIM_Constant>,<at 0.75 0.75 RCIM_Constant>,<at 0.9 1 RCIM_Constant>,<at 1 1.15 RCIM_Constant>,<at 1.1 1 RCIM_Constant>]}}}
     CurveIndex=&Multiply_Float(A=$Emitter.Age, B=4)
    )
   )
   &DynamicMaterialParameters(
    Index 0 Param 1=$System.Timeline.Float
    Module.Param0WriteEnabled=true
    Module.Param1WriteEnabled=true
    Module.Param2WriteEnabled=true
    Module.Param3WriteEnabled=true
   )
   /SetVariables(
    Particles.PivotOffset:Vector2f=&SelectVector2DFromArray(
     Direct Array Index=$User.PivotOffsetArrayNr
     Vector 2D Selection Array=DataInterface() {propertyValues {FloatData [<0.5 0.5>, <0.5 0.45>, <0.5 0.5>, <0.5 0.5>]}}
    )
    Particles.SpriteRotation:NiagaraFloat=&Lerp_Float(
     A=0
     B=15
     Alpha=&FloatFromCurve(
      FloatCurve=DataInterface() {propertyValues {Curve {keys [
          {arriveTangent=-3.5919, interpMode=RCIM_Cubic, leaveTangent=-3.5919, tangentMode=RCTM_User, time=0, value=1}
          {arriveTangent=-0.7811, interpMode=RCIM_Cubic, leaveTangent=-0.7811, tangentMode=RCTM_User, time=0.5, value=0}
         ]}}}
      CurveIndex=$Emitter.Age
      Scale Curve=&Sine(Normalized Angle=$Emitter.Age, Period=0.12, Scale=1, Bias=0)
     )
    )
   ) {isSetParameters=true}
  ]
  rendererClasses [&NiagaraSpriteRendererProperties]
  renderers [{rendererClass=&NiagaraSpriteRendererProperties}]
 }
 Noise {
  emitterUpdateScript [
   &EmitterState()
   &SpawnRate(
    Spawn Probability=&MakeCustomFloatFromBool(Bool=&SetBoolByFloatComparison(A=1, B=$System.Timeline.Float), True Float=0, False Float=0.05)
    SpawnRate=45
    Use Spawn Probability=true
   )
  ]
  particleSpawnScript [
   &InitializeParticle(
    Color Mode=::NewEnumerator0
    Lifetime Max=0.2
    Lifetime Min=0.1
    Lifetime Mode=::NewEnumerator1
    Sprite Size Max=<40 10>
    Sprite Size Min=<20 5>
    Sprite Size Mode=::NewEnumerator2
    UsePositionOffset=false
    Write Material Random=true
   )
   &DynamicMaterialParameters(
    Index 0 Param 1=&RandomRangeFloat(Minimum=0, Maximum=1)
    Module.Param0WriteEnabled=true
    Module.Param1WriteEnabled=true
    Module.Param2WriteEnabled=true
    Module.Param3WriteEnabled=true
   )
   &ShapeLocation(
    Box Size=&MakeCustomVectorFromBool(Bool=$User.NonUniformSizeSwitch, True Vector=<30 30 30>, False Vector=<25 25 15>)
    MaskSpawnGroup=false
    Offset=$User.Offset
    Offset Mode=::NewEnumerator0
    Override Seed=false
    Shape Primitive=::NewEnumerator2
    Surface Only=false
    UseOwnerScale=true
   )
   &CameraOffset(Camera Offset Amount=&Lerp_Float(
     A=15
     B=4
     Alpha=&FloatFromCurve(FloatCurve=DataInterface() {propertyValues {Curve {keys [<at 0 1>,<at 1 0>]}}}, CurveIndex=$Emitter.Scalability.DistanceFraction)
    )) {module=CameraOffset001}
   /SetVariables(Particles.PivotOffset:Vector2f=&MakeVector2D(X=&RandomRangeFloat(Minimum=0.25, Maximum=0.75), Y=0.5)) {isSetParameters=true}
  ]
  particleUpdateScript [
   &ParticleState()
   &ScaleSpriteSize(
    Non-Uniform Curve Sprite Scale=DataInterface() {propertyValues {
      XCurve {keys [
        {interpMode=RCIM_Cubic, tangentMode=RCTM_User, time=0, value=0}
        {arriveTangent=1.5, interpMode=RCIM_Cubic, leaveTangent=1.5, time=0.3, value=0.925}
        {interpMode=RCIM_Cubic, tangentMode=RCTM_User, time=0.6, value=0.9}
        <at 1 0 RCIM_Cubic>
       ]}
      YCurve {keys [
        {arriveTangent=9.0725, interpMode=RCIM_Cubic, leaveTangent=9.0725, tangentMode=RCTM_User, time=0, value=0}
        {arriveTangent=-0.0313, interpMode=RCIM_Cubic, leaveTangent=-0.0313, tangentMode=RCTM_User, time=0.1408, value=0.9847}
        {arriveTangent=-1.146, interpMode=RCIM_Cubic, leaveTangent=-1.146, time=0.8, value=0.825}
        {interpMode=RCIM_Cubic, leaveTangent=-0.0405, time=1, value=0}
       ]}
     }}
    Scale Sprite Size Mode=::NewEnumerator3
   )
  ]
  rendererClasses [&NiagaraSpriteRendererProperties]
  renderers [{rendererClass=&NiagaraSpriteRendererProperties}]
 }
 Orbs {
  emitterUpdateScript [
   &EmitterState(Life Cycle Mode=::NewEnumerator1, Loop Behavior=::NewEnumerator1, Loop Duration=0.5, UseLoopDelay=false)
   &SpawnBurst_Instantaneous(Spawn Count=16, Use Spawn Probability=false, UseLoopCountLimit=true)
   &SpawnBurst_Instantaneous(Spawn Count=15, Spawn Time=0.15, Use Spawn Probability=false, UseLoopCountLimit=false) {module=SpawnBurst_Instantaneous001}
  ]
  particleSpawnScript [
   /SetVariables(Particles.SizeLerp:NiagaraFloat=&FloatFromCurve(
     FloatCurve=DataInterface() {propertyValues {Curve {keys [
         {interpMode=RCIM_Cubic, tangentMode=RCTM_User, time=0, value=0}
         {arriveTangent=1.4324, interpMode=RCIM_Cubic, leaveTangent=1.4324, tangentMode=RCTM_User, time=0.87, value=0.5596}
         {interpMode=RCIM_Cubic, tangentMode=RCTM_User, time=1, value=1}
        ]}}}
     CurveIndex=&RandomRangeFloat(Minimum=0, Maximum=1)
    )) {isSetParameters=true}
   /SetVariables(Particles.Intensity:NiagaraFloat=&FloatFromCurve(
     FloatCurve=DataInterface() {propertyValues {Curve {keys [
         {arriveTangent=4.6871, interpMode=RCIM_Cubic, leaveTangent=4.6871, tangentMode=RCTM_User, time=0, value=0}
         {interpMode=RCIM_Cubic, tangentMode=RCTM_User, time=0.7, value=1}
        ]}}}
     CurveIndex=&MakeFloatFromLinearColor(LinearColor=<rgba 1 1 1 1>, Channel:ENiagaraLinearColor_Channels=::NewEnumerator3)
    )) {isSetParameters=true}
   /SetVariables(Particles.StateColorInt:NiagaraInt32=$User.StateColorInt) {isSetParameters=true, module=SetVariables_#1}
   &InitializeParticle(
    Color=&MultiplyLinearColor(
     Linear Color=&MakeCustomLinearColorFromBool(
      Bool=&SetBoolByFloatComparison(A=$Particles.SizeLerp, B=0.1)
      True Color=&SelectLinearColorFromArray(
       Array Sampling Mode=::NewEnumerator1
       Direct Array Index=$Particles.StateColorInt
       Color Selection Array=DataInterface() {propertyValues {ColorData [<rgba 0.4948 0 0.337 1>, <rgba 0.901 0.1369 0 1>, <rgba 0.143 0.3274 0.6 1>, <rgba 0.0878 0.5541 0.6 1>]}}
       Random Seed=?
      )
      False Color=&SelectLinearColorFromArray(
       Array Sampling Mode=::NewEnumerator1
       Direct Array Index=$Particles.StateColorInt
       Color Selection Array=DataInterface() {propertyValues {ColorData [<rgba 0 0.4116 0.526 1>, <rgba 0.4427 0.0116 0 1>, <rgba 0.4259 0.6 0.0191 1>, <rgba 0.6 0.0181 0.1339 1>]}}
       Random Seed=?
      )
     )
     Scale Factor=&MakeVector4_FromFloat(Float=$Particles.Intensity)
    )
    Lifetime Max=0.2
    Lifetime Min=0.1
    Lifetime Mode=::NewEnumerator1
    Sprite Rotation Mode=::NewEnumerator2
    Sprite Size Mode=::NewEnumerator3
    Uniform Sprite Size=&Lerp_Float(A=3, B=6, Alpha=$Particles.SizeLerp)
    UsePositionOffset=false
    Write Material Random=true
   )
   &ShapeLocation(
    MaskSpawnGroup=false
    Offset=$User.Offset
    Offset Mode=::NewEnumerator0
    Ring / Disc Distribution Mode:ENiagaraSphereDistributionMode=::NewEnumerator3
    Ring Radius=15
    Rotation Axis=&ReturnCameraVector(Camera Vectors=::NewEnumerator1, Coordinate Space=::Simulation)
    Rotation Mode=::NewEnumerator1
    Shape Primitive=::NewEnumerator4
    UseOwnerScale=true
   ) {module=ShapeLocation001}
   &CameraOffset(Camera Offset Amount=&RandomRangeFloat(Minimum=-5, Maximum=5))
   &AddVelocity(
    Module.Use Falloff Distance=false
    Origin Offset=$User.Offset
    Override Seed=false
    Velocity Mode=::NewEnumerator1
    Velocity Speed=&RandomRangeFloat(Minimum=55, Maximum=88)
   )
  ]
  particleUpdateScript [
   /SetVariables(Particles.ImpactDirection:Vector3f=&NormalizeVector(
     Vector To Normalize=&Subtract_Vector(
      A=Linked() {linkedVariable {name=Engine.Owner.Position, type=NiagaraPosition}}
      B=Linked() {linkedVariable {name=Particles.Position, type=NiagaraPosition}}
     )
     Safe Normalize=true
     Safe Fallback Threshold=0
     Safe Fallback Vector=<0 0 0.001>
    )) {isSetParameters=true}
   &ParticleState()
   &Drag(Drag=4, Use Linear Drag=true, Use Rotational Drag=false)
   &ScaleSpriteSize(
    Uniform Curve Scale=&Multiply_Float(
     A=$Particles.Intensity
     B=&Waveform(
      Global Phase=$Particles.Age
      "Waveform [1]"=::NewEnumerator12
      "[1] Frequency"=&RandomRangeFloat(Minimum=15, Maximum=35)
      "[1] Amplitude Min/Max"=<0.5 1.5>
     )
    )
    Uniform Curve Sprite Scale=DataInterface() {propertyValues {Curve {keys [
        {arriveTangent=10.8984, interpMode=RCIM_Cubic, leaveTangent=10.8984, tangentMode=RCTM_User, time=0, value=0}
        <at 0.2 1>
        {interpMode=RCIM_Cubic, tangentMode=RCTM_User, time=0.6, value=1}
        {arriveTangent=-5.0292, interpMode=RCIM_Cubic, leaveTangent=-5.0292, tangentMode=RCTM_User, time=1, value=0}
       ]}}}
   )
   /SetVariables(
    Particles.SpriteParams0:Vector4f=&MakeFloat4FromVectorAndFloat("Vector (RGB)"=&MakeVectorFromLinearColorRGB(Color=<rgba 0 0 0 1>), "Float (Alpha)"=0)
    Particles.SpriteParams1:Vector4f=&MakeFloat4FromVectorAndFloat("Vector (RGB)"=$Particles.ImpactDirection, "Float (Alpha)"=0)
   ) {isSetParameters=true}
   &SolveForcesAndVelocity(Clamp Velocity=false, Limit Acceleration=false)
  ]
  rendererClasses [&NiagaraSpriteRendererProperties]
  renderers [{rendererClass=&NiagaraSpriteRendererProperties}]
 }
 Ring {
  emitterUpdateScript [
   &EmitterState(Life Cycle Mode=::NewEnumerator1, Loop Behavior=::NewEnumerator1, Loop Duration=1, UseLoopDelay=false)
   &SpawnBurst_Instantaneous(
    Spawn Count=&SelectIntFromArray(Direct Array Index=$User.StateColorInt, Int Selection Array=DataInterface() {propertyValues {IntData [0,1,1,1]}})
    Spawn Time=0.1
    Use Spawn Probability=false
    UseLoopCountLimit=false
   )
  ]
  particleSpawnScript [&InitializeParticle(
    Color=&SelectLinearColorFromArray(
     Array Sampling Mode=::NewEnumerator1
     Direct Array Index=$User.StateColorInt
     Color Selection Array=DataInterface() {propertyValues {ColorData [<rgba 0 0 0 1>, <rgba 0.901 0.1369 0 1>, <rgba 0.143 0.3274 0.6 1>, <rgba 0.0878 0.5541 0.6 1>]}}
     Random Seed=?
    )
    Lifetime=0.2
    Mesh Scale Mode=::NewEnumerator3
    Mesh Uniform Scale=&SelectFloatFromArray(Direct Array Index=$User.StateColorInt, Float Selection Array=DataInterface() {propertyValues {FloatData [1,1.2,1.4,1.6]}})
    Position Offset=$User.Offset
    UsePositionOffset=true
    Write Material Random=true
   )]
  particleUpdateScript [
   &ParticleState()
   &ScaleMeshSize(Scale Factor=&Multiply_VectorByFloat(
     Vector=<1 1 1>
     Float=&FloatFromCurve(FloatCurve=DataInterface() {propertyValues {Curve {keys [
          {arriveTangent=1.235, interpMode=RCIM_Cubic, leaveTangent=1.235, tangentMode=RCTM_User, time=0, value=0.5}
          {arriveTangent=-0.0405, interpMode=RCIM_Cubic, leaveTangent=-0.0405, tangentMode=RCTM_User, time=1, value=1}
         ]}}})
    ))
   &ScaleColor(
    Scale Alpha=&FloatFromCurve(
     FloatCurve=DataInterface() {propertyValues {Curve {keys [
         {interpMode=RCIM_Cubic, tangentMode=RCTM_User, time=0, value=0}
         {interpMode=RCIM_Cubic, tangentMode=RCTM_User, time=0.5, value=1}
         {tangentMode=RCTM_User, time=1, value=0}
        ]}}}
     Scale Curve=25
    )
    ScaleA=true
    ScaleRGB=true
   )
   &DynamicMaterialParameters(
    Index 0 Param 1=1
    Index 0 Param 3=1
    Index 0 Param 4=&FloatFromCurve(
     FloatCurve=DataInterface() {propertyValues {Curve {keys [
         {interpMode=RCIM_Cubic, tangentMode=RCTM_User, time=0, value=0}
         {interpMode=RCIM_Cubic, tangentMode=RCTM_User, time=0.5, value=1}
         {tangentMode=RCTM_User, time=1, value=0}
        ]}}}
     Scale Curve=7
    )
    Module.Param0WriteEnabled=true
    Module.Param1WriteEnabled=true
    Module.Param2WriteEnabled=true
    Module.Param3WriteEnabled=true
   )
  ]
  rendererClasses [&NiagaraMeshRendererProperties]
  renderers [{rendererClass=&NiagaraMeshRendererProperties, rendererIndex=0}]
 }
 Splatter_Weakpoint {
  emitterUpdateScript [
   &EmitterState(Life Cycle Mode=::NewEnumerator1, Loop Behavior=::NewEnumerator1, Loop Duration=1, UseLoopDelay=false)
   &SpawnBurst_Instantaneous(
    Spawn Count=&SelectIntFromArray(Direct Array Index=$User.AdditionalElementInt, Int Selection Array=DataInterface() {propertyValues {IntData [0,0,5,0]}})
    Spawn Time=0.1
    Use Spawn Probability=false
    UseLoopCountLimit=false
   )
  ]
  particleSpawnScript [
   &InitializeParticle(
    Color=<rgba 0.0181 0.25 0.1551 1>
    Sprite Rotation Mode=::NewEnumerator2
    Sprite Size Mode=::NewEnumerator1
    Uniform Sprite Size Max=35
    Uniform Sprite Size Min=15
    UsePositionOffset=false
    Write Material Random=true
   )
   &DynamicMaterialParameters(
    Index 0 Param 1=&RandomRangeFloat(Minimum=2.5, Maximum=2)
    Index 0 Param 2=&RandomRangeFloat(Minimum=2.5, Maximum=2)
    Index 0 Param 3=&RandomRangeFloat(Minimum=1, Maximum=1.5)
    Index 0 Param 4=&RandomRangeFloat(Minimum=1, Maximum=1.5)
    Index 1 Param 1=&RandomRangeFloat(Minimum=0, Maximum=1)
    Index 1 Param 2=&RandomRangeFloat(Minimum=0, Maximum=1)
    Module.Param0WriteEnabled=true
    Module.Param0WriteEnabled1=true
    Module.Param1WriteEnabled=true
    Module.Param1WriteEnabled1=true
    Module.Param2WriteEnabled=true
    Module.Param2WriteEnabled1=true
    Module.Param3WriteEnabled=true
    Module.Param3WriteEnabled1=true
    Write Parameter Index 1=true
   )
   &ShapeLocation(
    MaskSpawnGroup=false
    Offset=$User.Offset
    Offset Mode=::NewEnumerator0
    Ring / Disc Distribution Mode:ENiagaraSphereDistributionMode=::NewEnumerator3
    Ring Radius=15
    Rotation Axis=&ReturnCameraVector(Camera Vectors=::NewEnumerator1, Coordinate Space=::Simulation)
    Rotation Mode=::NewEnumerator1
    Shape Primitive=::NewEnumerator4
    UseOwnerScale=true
   ) {module=ShapeLocation001}
   &CameraOffset(Camera Offset Amount=-10)
  ]
  particleUpdateScript [
   &ParticleState()
   &ScaleColor(
    Scale Alpha=&FloatFromCurve(FloatCurve=DataInterface() {propertyValues {Curve {keys [
         {interpMode=RCIM_Cubic, tangentMode=RCTM_User, time=0, value=1}
         <at 0.2 0.2>
         {interpMode=RCIM_Cubic, tangentMode=RCTM_User, time=0.6, value=0.2}
         {tangentMode=RCTM_User, time=1, value=1}
        ]}}})
    ScaleA=true
    ScaleRGB=true
   )
  ]
  rendererClasses [&NiagaraSpriteRendererProperties]
  renderers [{rendererClass=&NiagaraSpriteRendererProperties}]
 }
 Streak_Damage {
  emitterSpawnScript [/SetVariables() {isSetParameters=true}]
  emitterUpdateScript [
   &EmitterState(Life Cycle Mode=::NewEnumerator1, Loop Behavior=::NewEnumerator1, Loop Duration=0.4, UseLoopDelay=false)
   &SpawnBurst_Instantaneous(
    Spawn Count=&SelectIntFromArray(Direct Array Index=$User.AdditionalElementInt, Int Selection Array=DataInterface() {propertyValues {IntData [0,6,0,6]}})
    Spawn Time=0.05
    Use Spawn Probability=false
    UseLoopCountLimit=false
   )
  ]
  particleSpawnScript [
   &InitializeParticle(
    Color Mode=::NewEnumerator0
    Lifetime Max=0.5
    Lifetime Min=0.25
    Lifetime Mode=::NewEnumerator1
    Sprite Size Max=<1 16>
    Sprite Size Min=<1 10>
    Sprite Size Mode=::NewEnumerator2
    UsePositionOffset=false
    Write Material Random=true
   )
   &ApplyInitialForces()
   &DynamicMaterialParameters(
    Index 0 Param 1=&RandomRangeFloat(Minimum=0.5, Maximum=0.75)
    Index 0 Param 2=&RandomRangeFloat(Minimum=0.15, Maximum=0.4)
    Index 0 Param 3=&RandomRangeFloat(Minimum=0, Maximum=1)
    Module.Param0WriteEnabled=true
    Module.Param1WriteEnabled=true
    Module.Param2WriteEnabled=true
    Module.Param3WriteEnabled=true
   )
   &ShapeLocation(
    MaskSpawnGroup=false
    Offset=$User.Offset
    Offset Mode=::NewEnumerator0
    Override Seed=false
    Ring Radius=&RandomRangeFloat(Minimum=12, Maximum=18)
    Rotation Axis=&ReturnCameraVector(Camera Vectors=::NewEnumerator1, Coordinate Space=::Simulation)
    Rotation Mode=::NewEnumerator1
    Shape Primitive=::NewEnumerator4
    UseOwnerScale=true
   )
   &AddVelocity(
    Module.Use Falloff Distance=false
    Origin Offset=$User.Offset
    Override Seed=false
    Velocity Mode=::NewEnumerator1
    Velocity Speed=&RandomRangeFloat(Minimum=33, Maximum=44)
   )
   /SetVariables(Particles.VisibilityTag=$User.StreakVisInt) {isSetParameters=true}
  ]
  particleUpdateScript [
   &ParticleState()
   &ScaleSpriteSize(Uniform Curve Sprite Scale=DataInterface() {propertyValues {Curve {keys [{arriveTangent=6.0076, interpMode=RCIM_Cubic, leaveTangent=6.0076, tangentMode=RCTM_User, time=0, value=0}, <at 0.2 1 RCIM_Cubic>]}}})
   &Drag(Drag=5, Use Linear Drag=true, Use Rotational Drag=true)
   &SolveForcesAndVelocity(Clamp Velocity=false, Limit Acceleration=false)
   &ScaleColor(
    Scale Alpha=&FloatFromCurve(FloatCurve=DataInterface() {propertyValues {Curve {keys [{tangentMode=RCTM_User, time=0.5, value=0}, {tangentMode=RCTM_User, time=1, value=1}]}}})
    ScaleA=true
    ScaleRGB=true
   )
  ]
  rendererClasses [&NiagaraSpriteRendererProperties]
  renderers [
   {rendererClass=&NiagaraSpriteRendererProperties}
   {rendererClass=&NiagaraSpriteRendererProperties, rendererIndex=1}
  ]
 }
}
summary {
 systemName=NS_Backpack_RoseDust
 emitters [
  {emitterName=Graphic, rendererClasses [&NiagaraSpriteRendererProperties]}
  {emitterName=Orbs, rendererClasses [&NiagaraSpriteRendererProperties]}
  {emitterName=GlitchShapes, rendererClasses [&NiagaraMeshRendererProperties]}
  {emitterName=Splatter_Weakpoint, rendererClasses [&NiagaraSpriteRendererProperties]}
  {emitterName=Streak_Damage, rendererClasses [&NiagaraSpriteRendererProperties]}
  {emitterName=Ring, rendererClasses [&NiagaraMeshRendererProperties]}
  {emitterName=Digital_Bits, simTarget=GPUComputeSim, rendererClasses [&NiagaraSpriteRendererProperties]}
  {emitterName=Noise, rendererClasses [&NiagaraSpriteRendererProperties]}
 ]
 userVariables [
  User.AdditionalElementInt {type=NiagaraInt32}
  User.bIsFrontEnd {type=NiagaraBool}
  User.bIsFrontEndPreview {type=NiagaraBool}
  User.ColorHigh {type=LinearColor, description="", defaultValue {struct=LinearColor, value=<rgba 0 1 1 1>}}
  User.ColorLow {type=LinearColor, description="", defaultValue {struct=LinearColor, value=<rgba 0.7294 0 0.1804 1>}}
  User.EffectChannelPicker {type=Vector3f, description="", defaultValue {struct=Vector3f, value=<1 0 0>}}
  User.IconColor {type=LinearColor, description="", defaultValue {struct=LinearColor, value=<rgba 1 1 1 1>}}
  User.Material {
   type=MaterialInterface
   description=""
   defaultValue {
    struct=NiagaraExt_VariableValue_Object
    value {
     object=/BRCosmetics/Accessories/FORT_Backpacks/Backpack_F_MED_Rosedust/Materials/MI_Rosedust_Meanwhile_EN_MI
     objectClass=/Script/Engine.MaterialInterface
    }
   }
  }
  User.NonUniformSizeSwitch {type=NiagaraBool}
  User.Offset {type=Vector3f, description="", defaultValue {struct=Vector3f, value=<0 -7 -1>}}
  User.PivotOffsetArrayNr {type=NiagaraInt32}
  User.StateColorInt {type=NiagaraInt32}
  User.StreakVisInt {type=NiagaraInt32}
 ]
}
