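# Unlit masked Rosedust backpack material, usable on Niagara sprites, that renders a glitchy, chromatic-aberration image with HDR-boosted emissive.
# Separately sampled RGB channels get hue-shifted and lighten-blended over the base sample, then conditionally darkened; low quality uses the plain sample.
# UV helpers build flipbook distortion, scrolling noise glitches, stepped jitter and camera-depth-faded channel offsets.
# Further helpers derive light-dependent halftone dots and virtual-plane coordinates facing camera and light.
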Material /BRCosmetics/Accessories/FORT_Backpacks/Backpack_F_MED_Rosedust/Materials/M_RosedustBackpack
graph {
 properties {bUsedWithNiagaraSprites=true, blendMode=BLEND_Masked, pixelDepthOffsetMode=PDOM_Legacy, shadingModel=MSM_Unlit, usedShadingModels=Unlit}
 nodes [
  n1 = Multiply(A=NamedRerouteUsage()) {constB=0.1}
  n2 = NamedRerouteUsage()
  n3 = TextureObjectParameter() {parameterName=Tex, texture=/BRCosmetics/Accessories/FORT_Backpacks/Backpack_F_MED_Rosedust/Textures/Meanwhile/T_RoseDust_Meanwhile}
  n4 = TextureSample(UVs=NamedRerouteUsage(), Tex=@n3)
  n5 = TextureSample(UVs=NamedRerouteUsage(), Tex=@n3)
  n6 = TextureSample(UVs=NamedRerouteUsage(), Tex=@n3)
  n7 = TextureSample(UVs=NamedRerouteUsage(), Tex=@n3)
  n8 = MaterialFunctionCall!(
   Blend=MaterialFunctionCall(
    Hue Shift Percentage=LinearInterpolate(B=@n1, Alpha=@n2)
    Texture=MaterialFunctionCall(X=@n4.R, Y=@n5.G, Z=@n6.B) {materialFunction=&MakeFloat3}.Result
   ) {materialFunction=&HueShift}.Result
   Base=@n7.RGB
  ) {materialFunction=/Engine/Functions/Engine_MaterialFunctions03/Blends/Blend_Lighten}
  n9 = QualitySwitch(Default=@n8.Result, Low=@n7.RGB, High=@n8.Result, Medium=@n8.Result, Epic=@n8.Result)
  n10 = ScalarParameter() {defaultValue=0.5, parameterName=FrontEndColorCorrect}
  n11 = NamedRerouteUsage()
  n12 = LinearInterpolate(A=@n9, B=Multiply(A=@n9, B=LinearInterpolate(A=@n10, Alpha=@n11) {constA=1}) {constB=0.33}, Alpha=Saturate(None=NamedRerouteUsage()))
  MaterialFunctionCall(Emissive=MaterialFunctionCall(
    LDR Input=QualitySwitch(Default=@n12, Low=@n9, High=@n12, Medium=@n12, Epic=@n12)
    HDR Intensity Multiplier=ScalarParameter() {defaultValue=1, parameterName=HDR Contrast}
    Derive HDR Power=ScalarParameter() {defaultValue=1, parameterName=DeriveHDRPower}
   ) {materialFunction=&DeriveHDRfromLDR}.RemappedRange) {materialFunction=&MF_FX_EmissiveAdjustments}
  n13 = Constant() {r=0.94}
  n14 = TextureCoordinate()
  n15 = MaterialFunctionCall(Float2=Abs(None=Add(A=@n14) {constB=-0.5})) {materialFunction=&BreakOutFloat2Components}
  n16 = Multiply(A=Max(A=@n15.R, B=@n15.R)) {constB=2}
  Multiply(A=@n7.A, B=OneMinus(None=SmoothStep(Min=@n13, Max=@n13, Value=@n16)))
  n17 = NamedRerouteUsage()
  n18 = Constant()
  n19 = LinearInterpolate(
   B=ScalarParameter() {defaultValue=0.05, parameterName=ChromaMaxOffset}
   Alpha=LinearInterpolate(B=NamedRerouteUsage(), Alpha=NamedRerouteUsage())
  ) {constA=0.01, constAlpha=1, constB=0.1}
  n20 = ScalarParameter() {defaultValue=1, parameterName=FlipbookdistortionStrenght}
  n21 = LinearInterpolate(
   A=MaterialFunctionCall(Fade Length=Constant() {r=200}, Fade Offset=Constant()) {materialFunction=&CameraDepthFade}.Result
   Alpha=@n20
  ) {constB=0}
  n22 = LinearInterpolate(A=Multiply(A=@n19) {constB=-1}, Alpha=@n21) {constA=-0.01, constB=0}
  NamedRerouteDeclaration(None=Add(A=@n17, B=AppendVector(A=@n18, B=@n22))) {name=OffsetR, nodeColor=<rgba 0.4531 1 0 1>}
  NamedRerouteDeclaration(None=Add(A=@n17, B=AppendVector(A=@n22, B=@n18))) {name=OffsetG, nodeColor=<rgba 1 0 0.3516 1>}
  n23 = TextureObject() {
   samplerType=SAMPLERTYPE_LinearColor
   texture=/BRCosmetics/Accessories/FORT_Backpacks/Backpack_F_MED_Rosedust/Textures/T_Rosedust_Meanwhile_Distortion_N
  }
  n24 = NamedRerouteUsage()
  n25 = MaterialFunctionCall(Float3=MaterialFunctionCall!(
    Texture=@n23
    UVs=@n24
    SubImages=Constant2Vector() {g=2, r=2}
    Frame=DynamicParameter() {defaultValue=<rgba 0 1 1 1>, paramNames [Frames,Param2,Param3,Param4]}.Frames
   ) {materialFunction=/Engine/Functions/Engine_MaterialFunctions03/Texturing/SubUV_Function}.RGB) {materialFunction=&BreakOutFloat3Components}
  NamedRerouteDeclaration(None=LinearInterpolate(
    A=@n24
    B=MaterialFunctionCall(X=@n25.R, Y=@n25.R) {materialFunction=&MakeFloat2}.Result
    Alpha=@n20
   ) {constAlpha=1}) {name=Flipbookdistortion, nodeColor=<rgba 1 0.75 0 1>}
  NamedRerouteDeclaration(None=@n25.B) {name=HueShiftMap, nodeColor=<rgba 0.0312 0 1 1>}
  NamedRerouteDeclaration(None=ScalarParameter() {parameterName=ChromaAmount}) {name=ChromaAmount, nodeColor=<rgba 1 0.3281 0 1>}
  n26 = LinearInterpolate(A=@n19, Alpha=@n21) {constA=0.01, constB=0}
  NamedRerouteDeclaration(None=Add(A=@n17, B=AppendVector(A=@n26, B=@n18))) {name=OffsetB, nodeColor=<rgba 0 1 0.0391 1>}
  n27 = MaterialFunctionCall() {materialFunction=&CameraDirectionVector}
  n28 = Multiply(A=Custom() {code="return View.DirectionalLightDirection;", description=DirectionalLightDirection, inputs []}, B=@n27.Result)
  n29 = Add(A=Multiply(B=Normalize(VectorInput=@n28)) {constA=0.25}, B=@n27.Result)
  n30 = Multiply(A=@n27.Result) {constB=50}
  n31 = Add(A=ActorPositionWS(), B=Constant3Vector() {constant=<rgba 0 0 90 1>})
  n32 = Constant4Vector() {constant=<rgba 64 128 45 90>}
  n33 = InverseLinearInterpolate(
   A=@n32.B
   B=@n32.B
   Value=MaterialFunctionCall() {materialFunction=/Engine/Functions/Engine_MaterialFunctions02/Utility/FOV}.FOV angle in degrees
  )
  NamedRerouteDeclaration(None=MaterialFunctionCall(Plane Normal Axis=Normalize(VectorInput=@n29), Plane Center=Add(A=@n30, B=@n31), UV Scale=LinearInterpolate(A=@n32.R, B=@n32.R, Alpha=@n33)) {materialFunction=&VirtualPlaneCoordinates}.UVs) {name=ObjectAlighedScreenPosition, nodeColor=<rgba 1 0.867 0 1>}
  n34 = Subtract(A=Distance(A=ActorPositionWS(), B=CameraPositionWS())) {constB=100}
  NamedRerouteDeclaration(None=Saturate(None=Divide(A=@n34) {constB=400})) {name=DistanceBlending, nodeColor=<rgba 0.4766 1 0 1>}
  n35 = Time()
  n36 = Frac(None=@n35)
  n37 = Divide(A=Floor(None=Multiply(A=@n36) {constB=5})) {constB=5.4}
  n38 = OneMinus(None=@n37)
  n39 = Floor(None=Multiply(A=Frac(None=@n35)) {constB=2})
  n40 = Frac(None=@n35)
  n41 = Divide(A=Floor(None=Multiply(A=@n40) {constB=2})) {constB=3.4}
  n42 = OneMinus(None=@n41)
  n43 = AppendVector(
   A=LinearInterpolate(Alpha=LinearInterpolate(A=@n38, B=@n37, Alpha=@n39)) {constA=-0.1, constB=0.1}
   B=LinearInterpolate(Alpha=LinearInterpolate(A=@n42, B=@n41, Alpha=@n39)) {constA=-0.1, constB=0.1}
  )
  n44 = ScalarParameter() {defaultValue=0.1, parameterName=OffsetStrenght}
  n45 = VectorParameter() {parameterName=EffectsPicker, defaultValue=<rgba 0 0 0 1>}
  n46 = MaterialFunctionCall(Float3=MaterialFunctionCall(X=@n45.R, Y=@n45.R, Z=@n45.R) {materialFunction=&MakeFloat3}.Result) {materialFunction=&BreakOutFloat3Components}
  n47 = Constant3Vector() {constant=<rgba 1 1 1 0>}
  n48 = MaterialFunctionCall(Float3=MaterialFunctionCall(X=@n47.R, Y=@n47.R, Z=@n47.R) {materialFunction=&MakeFloat3}.Result) {materialFunction=&BreakOutFloat3Components}
  NamedRerouteDeclaration(None=LinearInterpolate(B=Multiply(A=@n43, B=@n44) {constB=0.1}, Alpha=DotProduct(A=@n46.R, B=@n48.R))) {name=RandomOffset, nodeColor=<rgba 0.1719 1 0 1>}
  n49 = NamedRerouteUsage()
  n50 = ScalarParameter() {defaultValue=100, group=1_RoseDust, parameterName=DotsAmount, sliderMax=512, sortPriority=20}
  n51 = LinearInterpolate(A=@n50, B=Multiply(A=@n50) {constB=0.5}, Alpha=NamedRerouteUsage())
  n52 = Frac(None=Rotator(Coordinate=Multiply(A=@n49, B=@n51), Time=Constant() {r=1}) {speed=0.75})
  n53 = OneMinus(None=Distance(A=Add(A=@n52) {constB=-0.5}, B=Constant()))
  n54 = AtmosphericLightVector()
  n55 = Multiply(A=DotProduct(A=Normalize(VectorInput=@n54), B=VertexNormalWS())) {constB=-1}
  NamedRerouteDeclaration(None=SmoothStep(Min=@n53, Max=@n53, Value=Saturate(None=@n55))) {name=DotBlend, nodeColor=<rgba 0 0.4844 1 1>}
  NamedRerouteDeclaration(None=@n21) {name=CameraOffset, nodeColor=<rgba 0 1 0.3672 1>}
  n56 = TextureCoordinate()
  n57 = Sine(None=@n38)
  n58 = LinearInterpolate(B=LinearInterpolate(Alpha=@n57) {constA=0.96, constB=1.01}, Alpha=DotProduct(A=@n46.G, B=@n48.G)) {constA=1}
  n59 = Multiply(A=@n57) {constB=0.3}
  n60 = DotProduct(A=@n46.B, B=@n48.B)
  NamedRerouteDeclaration(None=Rotator(
    Coordinate=MaterialFunctionCall(UVs=@n56, Texture Scale=@n58) {materialFunction=&ScaleUVsByCenter}.UVs
    Time=LinearInterpolate(B=@n59, Alpha=@n60)
   )) {name=UV, nodeColor=<rgba 1 0 0.0937 1>}
  n61 = TextureSample(UVs=Panner(Coordinate=TextureCoordinate() {uTiling=0}) {speedX=0.5, speedY=0.5}) {texture=/Game/Athena/UI/Challenges/Art/T_UI_DataChip_Noise}
  n62 = AppendVector(A=@n61.R, B=@n61.R)
  NamedRerouteDeclaration(None=LinearInterpolate(A=Multiply(A=@n62) {constB=0.01}, Alpha=NamedRerouteUsage()) {constB=0}) {name=CloseCameraDistortion, nodeColor=<rgba 0 0.4844 1 1>}
  n63 = ScalarParameter() {parameterName=GlitchStrenght}
  n64 = TextureCoordinate() {vTiling=0.5}
  n65 = Sine(None=MaterialFunctionCall(UV Channel=Panner(Coordinate=@n64) {speedY=1}) {materialFunction=&LinearGradient}.VGradient)
  n66 = TextureCoordinate() {vTiling=0.6}
  n67 = Sine(None=MaterialFunctionCall(UV Channel=Panner(Coordinate=@n66) {speedY=3}) {materialFunction=&LinearGradient}.VGradient)
  n68 = Add(A=Ceil(None=Sine(None=@n65)), B=Ceil(None=Sine(None=@n67)))
  n69 = Panner(Coordinate=TextureCoordinate() {uTiling=0.2, vTiling=0.2}) {speedX=0.5, speedY=0.5}
  n70 = Multiply(A=Time())
  n71 = Clamp(
   None=Multiply(A=TextureSample(UVs=@n69) {texture=&T_UI_DataChip_Noise}.R) {constB=0.5}
   Min=Cosine(None=Sine(None=@n70) {period=2}) {period=0.5}
  )
  n72 = Add(A=LinearInterpolate(B=Divide(A=@n63) {constB=100}, Alpha=Multiply(A=@n68, B=@n71)) {constB=0.01}, B=NamedRerouteUsage())
  n73 = NamedRerouteUsage()
  NamedRerouteDeclaration(None=Add(A=Add(A=@n72, B=@n73), B=NamedRerouteUsage())) {name=Distortion, nodeColor=<rgba 0 0.9062 1 1>}
 ]
}
