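# Unlit masked Niagara sprite material for the Rosedust backpack, drawing a black-and-white glitch static pattern.
# A noise texture, panned by the RandomOffset dynamic parameter instead of game time, warps tiled UVs.
# Two overlaid grids of sphere-mask dots are hard-thresholded through equal-edged SmoothSteps and passed to MF_FX_EmissiveAdjustments.
# The same noise distorts a centered box shape, thresholded and inverted into a ragged-edged cutout mask.
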Material /BRCosmetics/Accessories/FORT_Backpacks/Backpack_F_MED_Rosedust/Materials/M_Rosedust_GlitchStaticNoise
graph {
 properties {bUsedWithNiagaraSprites=true, blendMode=BLEND_Masked, pixelDepthOffsetMode=PDOM_Legacy, shadingModel=MSM_Unlit, usedShadingModels=Unlit}
 nodes [
  n1 = Constant() {r=0.3}
  n2 = TextureSample(UVs=Panner(
    Coordinate=TextureCoordinate() {uTiling=0.1, vTiling=0.5}
    Time=DynamicParameter() {paramNames [RandomOffset,Param2,Param3,Param4]}.RandomOffset
   ) {speedX=0.5, speedY=0.5}) {texture=/Game/Athena/UI/Challenges/Art/T_UI_DataChip_Noise}
  n3 = Multiply(A=Add(A=TextureCoordinate() {vTiling=0.25}, B=Multiply(A=@n2.R) {constB=0.1}), B=Constant() {r=22}) {constB=6}
  n4 = Abs(None=Add(A=@n3) {constB=-0.5})
  n5 = Constant2Vector() {g=0.5, r=0.5}
  n6 = Constant() {r=0.4}
  n7 = Constant()
  n8 = Add(A=SphereMask(A=Frac(None=@n4), B=@n5, Radius=@n6, Hardness=@n7), B=SphereMask(A=Frac(None=@n3), B=@n5, Radius=@n6, Hardness=@n7))
  MaterialFunctionCall(Emissive=LinearInterpolate(A=Constant(), B=Constant() {r=1}, Alpha=SmoothStep(Min=@n1, Max=@n1, Value=@n8) {constMax=0.5, constMin=0.5})) {materialFunction=&MF_FX_EmissiveAdjustments}
  MaterialFunctionCall(Texture Dimensions=MaterialFunctionCall() {materialFunction=&ScreenResolution}.Visible Resolution) {materialFunction=&ScreenAlignedUVs}
  n9 = Constant() {r=0.15}
  n10 = Abs(None=Add(A=TextureCoordinate()) {constB=-0.5})
  n11 = Constant() {r=0.5}
  n12 = Add(A=@n2.R, B=@n11)
  n13 = MaterialFunctionCall(Float2=Add(A=Multiply(A=@n10, B=@n12), B=Multiply(A=@n2.R, B=@n11))) {materialFunction=&BreakOutFloat2Components}
  OneMinus(None=SmoothStep(Min=@n9, Max=@n9, Value=Max(A=@n13.R, B=@n13.R)))
 ]
}
