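# Translucent unlit Niagara sprite material for a paint splatter effect on the Rosedust backpack cosmetic.
# Two panned noise samples multiply inside a sphere mask whose radius is wobbled by Perlin noise, forming the splatter silhouette.
# Particle alpha sets the SmoothStep erosion threshold, and particle color tints the result into MF_FX_EmissiveAdjustments.
# Dynamic parameters drive tiling and random offsets, but the TileY and Erosion outputs are unused.
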Material /BRCosmetics/Accessories/FORT_Backpacks/Backpack_F_MED_Rosedust/Materials/M_Rosedust_PaintSplatter
graph {
 properties {bUsedWithNiagaraSprites=true, blendMode=BLEND_Translucent, pixelDepthOffsetMode=PDOM_Legacy, shadingModel=MSM_Unlit, usedShadingModels=Unlit}
 nodes [
  n1 = ParticleColor()
  n2 = Add(A=@n1.A) {constB=0.1}
  n3 = TextureCoordinate()
  n4 = DynamicParameter() {defaultValue=<rgba 2 2 1 1>, paramNames [Noise1_TileX,Noise1_TileY,Noise2_TileX,Noise2_TileY]}
  n5 = AppendVector(A=@n4.Noise1_TileX, B=@n4.Noise1_TileX)
  n6 = DynamicParameter() {parameterIndex=1, paramNames [NoiseRandomOffset, SilhouetteRandomOFfset, Erosion, Param4]}
  n7 = TextureSample(UVs=Panner(Coordinate=Multiply(A=@n3, B=@n5), Time=@n6.NoiseRandomOffset) {speedX=1, speedY=1}) {samplerType=SAMPLERTYPE_LinearColor, texture=/Game/Environments/Foliage/Textures/T_NoiseMask02_M}
  n8 = TextureCoordinate()
  n9 = AppendVector(A=@n4.Noise2_TileX, B=@n4.Noise2_TileX)
  n10 = TextureSample(UVs=Panner(Coordinate=Multiply(A=@n8, B=@n9), Time=@n6.NoiseRandomOffset) {speedX=1, speedY=1}) {samplerType=SAMPLERTYPE_LinearColor, texture=&T_NoiseMask02_M}
  n11 = Constant() {r=0.3}
  n12 = Multiply(A=TextureSample(UVs=Panner(Time=@n6.SilhouetteRandomOFfset) {speedX=-1, speedY=-1}) {texture=/Game/Items/Traps/Textures/T_PerlinRGB_01_M}.R) {constB=0.2}
  n13 = Multiply(
   A=Multiply(A=Multiply(A=@n7.R, B=@n10.R)) {constB=2}
   B=SphereMask(A=TextureCoordinate(), B=Constant() {r=0.5}, Radius=Add(A=@n11, B=@n12), Hardness=Constant())
  )
  MaterialFunctionCall(Emissive=Multiply(A=@n1.RGB, B=SmoothStep(Min=@n1.A, Max=@n2, Value=@n13) {constMin=1})) {materialFunction=&MF_FX_EmissiveAdjustments}
 ]
}
