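# Translucent Niagara sprite material for the Rosedust backpack, drawing a two-color, noise-broken paint streak.
# Two slightly offset pairs of crossed sphere masks have smoke noise subtracted, then get smoothstepped; their difference blends Color2 over Color1 for emissive.
# Dynamic parameters tile and offset the noise, though Xtile drives both axes and Ytile is unused.
# Particle alpha thresholds the combined shape, weighted by a vertical gradient and noise, producing an erosion-style dissolve.
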Material /BRCosmetics/Accessories/FORT_Backpacks/Backpack_F_MED_Rosedust/Materials/M_Rosedust_Paintstreak
graph {
 properties {bUsedWithNiagaraSprites=true, blendMode=BLEND_Translucent, pixelDepthOffsetMode=PDOM_Legacy, usedShadingModels=Default Lit}
 nodes [
  n1 = VectorParameter() {parameterName=Color1, defaultValue=<rgba 0.9301 0.1413 0 1>}
  n2 = SphereMask(A=TextureCoordinate() {uTiling=0}, B=Constant2Vector() {g=0.45}, Radius=Constant() {r=0.45}, Hardness=Constant())
  n3 = SphereMask(A=TextureCoordinate() {vTiling=0}, B=Constant2Vector() {r=0.45}, Radius=Constant() {r=0.45}, Hardness=Constant() {r=0.5})
  n4 = TextureCoordinate() {vTiling=0.5}
  n5 = DynamicParameter() {paramNames [Xtile,Ytile,RandomOffset,Param4]}
  n6 = AppendVector(A=@n5.Xtile, B=@n5.Xtile)
  n7 = TextureSample(UVs=Panner(Coordinate=Multiply(A=@n4, B=@n6), Time=@n5.RandomOffset) {speedX=1, speedY=1}) {texture=/Game/Effects/Fort_Effects/Textures/Tile/T_Inky_Smoke_Tile2}
  n8 = SmoothStep(Value=Subtract(A=Multiply(A=@n2, B=@n3), B=@n7.R)) {constMax=0.2, constMin=0.1}
  n9 = TextureCoordinate() {uTiling=0}
  n10 = Constant2Vector() {g=0.55}
  n11 = Constant() {r=0.45}
  n12 = Constant()
  n13 = TextureCoordinate() {vTiling=0}
  n14 = Constant2Vector() {r=0.55}
  n15 = Constant() {r=0.45}
  n16 = Constant() {r=0.5}
  n17 = Subtract(A=Multiply(A=SphereMask(A=@n9, B=@n10, Radius=@n11, Hardness=@n12), B=SphereMask(A=@n13, B=@n14, Radius=@n15, Hardness=@n16)), B=@n7.R)
  n18 = Saturate(None=Subtract(A=SmoothStep(Value=@n17) {constMax=0.2, constMin=0.1}, B=@n8))
  MaterialFunctionCall(Emissive=LinearInterpolate(A=Multiply(A=@n1.RGB, B=@n8), B=VectorParameter() {parameterName=Color2, defaultValue=<rgba 0.4452 0.0137 0 1>}.RGB, Alpha=@n18)) {materialFunction=&MF_FX_EmissiveAdjustments}
  n19 = ParticleColor()
  n20 = OneMinus(None=ComponentMask(None=TextureCoordinate()) {g=true})
  n21 = Add(A=@n7.R) {constB=0.25}
  SmoothStep(Min=@n19.A, Max=Add(A=@n19.A) {constB=0.05}, Value=Multiply(A=Add(A=@n8, B=@n18), B=Multiply(A=@n20, B=@n21)))
 ]
}
